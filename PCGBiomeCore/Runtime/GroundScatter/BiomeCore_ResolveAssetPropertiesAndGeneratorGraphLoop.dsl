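; PCG loop subgraph for Biome Core that resolves biome asset properties and their generator graph from asset object paths into attribute sets.
; It drops empty RuntimeAssets paths, reads the BiomeAssets array, and copies GeneratorGraph onto it.
; Parameter-toggled branches extract option sub-structs, building a Transform and renaming assembly AllowCollision before merging them back as attributes.
; Only Enabled entries are output; empty input bypasses processing.
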(graph BiomeCore_ResolveAssetPropertiesAndGeneratorGraphLoop
  (params (BiomeAssetProperty :type Name :default Assets) (ExtractDebugOptions :type Bool :default true) (ExtractAssetOptions :type Bool :default true) (ExtractMeshOptions :type Bool :default true) (ExtractAssemblyOptions :type Bool :default true) (ExtractFilterOptions :type Bool :default true) (ExtractRuntimeOptions :type Bool :default true))
  (bind GetEntriesCount_32 (Get%20Element%20Count))
  (bind AttributeCast_0 (Attribute%20Cast :OutputType Boolean))
  (bind Branch_30 (Branch))
  (bind AttributeFilter_0 (Filter%20Attribute%20Elements :AttributeTypes (struct :Type SoftObjectPath) :bUseConstantThreshold true :Operator NotEqual :TargetAttribute RuntimeAssets :ThresholdAttribute @Last))
  (bind Branch_5 (Branch :bOutputToB-param ExtractAssetOptions))
  (bind Branch_1 (Branch :bOutputToB-param ExtractDebugOptions))
  (bind Branch_18 (Branch :bOutputToB-param ExtractMeshOptions))
  (bind Branch_21 (Branch :bOutputToB-param ExtractFilterOptions))
  (bind Branch_9 (Branch :bOutputToB-param ExtractRuntimeOptions))
  (bind Branch_7 (Branch :bOutputToB-param ExtractAssemblyOptions))
  (bind AttributeRename_3 (Attribute%20Rename :AttributeToRename AllowCollision :NewAttributeName AssemblyAllowCollision))
  (bind AddAttribute_1 (Add%20Attribute :AttributeTypes (struct :Type SoftObjectPath) :OutputTarget GeneratorGraph))
  (bind GetPropertyFromObjectPath_3 (Get%20Property%20From%20Object%20Path :bSanitizeOutputAttributeName false :OutputAttributeName GeneratorGraph :PropertyName GeneratorGraph))
  (bind CopyAttributes_0 (Copy%20Attributes :bCopyAllAttributes true :InputSource @Last))
  (bind MergeAttributes_16 (Merge%20Attributes))
  (bind MakeTransformAttribute_3 (Make%20Transform%20Attribute :InputSource1 Translation :InputSource2 Rotation :InputSource3 Scale :OutputTarget Transform))
  (bind DeleteAttributes_4 (Delete%20Attributes :bTokenizeOnWhiteSpace true :SelectedAttributes "Translation,Rotation,Scale"))
  (bind AddAttribute_41 (Add%20Attribute :bCopyAllAttributes true))
  (bind AttributeFilter_45 (Filter%20Attribute%20Elements :AttributeTypes (struct :Type Boolean :BoolValue true) :bUseConstantThreshold true :bUseSpatialQuery true :bWarnOnDataMissingAttribute false :Operator Equal :TargetAttribute Enabled))
  ;[0 0 0] Assets Base Structures
  (bind GetPropertyFromObjectPath_0 (Get%20Property%20From%20Object%20Path :bSanitizeOutputAttributeName false :OutputAttributeName None :PropertyName BiomeAssets :InputSource-param BiomeAssetProperty))
  ;[0 0 0] Asset Options Sub-Structures
  (bind GetPropertyFromObjectPath_1 (Get%20Property%20From%20Object%20Path :bSanitizeOutputAttributeName false :OutputAttributeName None :PropertyName BiomeAssets.AssetOptions :InputSource-param BiomeAssetProperty))
  (bind GetPropertyFromObjectPath_6 (Get%20Property%20From%20Object%20Path :bSanitizeOutputAttributeName false :OutputAttributeName None :PropertyName BiomeAssets.DebugOptions :InputSource-param BiomeAssetProperty))
  (bind GetPropertyFromObjectPath_8 (Get%20Property%20From%20Object%20Path :bSanitizeOutputAttributeName false :OutputAttributeName None :PropertyName BiomeAssets.MeshOptions :InputSource-param BiomeAssetProperty))
  (bind GetPropertyFromObjectPath_10 (Get%20Property%20From%20Object%20Path :bSanitizeOutputAttributeName false :OutputAttributeName None :PropertyName BiomeAssets.FilterOptions :InputSource-param BiomeAssetProperty))
  (bind GetPropertyFromObjectPath_12 (Get%20Property%20From%20Object%20Path :bSanitizeOutputAttributeName false :OutputAttributeName None :PropertyName BiomeAssets.RuntimeOptions :InputSource-param BiomeAssetProperty))
  (bind GetPropertyFromObjectPath_2 (Get%20Property%20From%20Object%20Path :bSanitizeOutputAttributeName false :OutputAttributeName None :PropertyName BiomeAssets.AssemblyOptions :InputSource-param BiomeAssetProperty))
  (bind MergeAttributes_28 (Merge%20Attributes))
  (bind MergeAttributes_18 (Merge%20Attributes))
  (bind MergeAttributes_20 (Merge%20Attributes))
  (bind MergeAttributes_22 (Merge%20Attributes))
  (bind MergeAttributes_24 (Merge%20Attributes))
  (bind MergeAttributes_26 (Merge%20Attributes))
  (connect GetPropertyFromObjectPath_0 Out AddAttribute_1 In)
  (connect AttributeRename_3 Out AddAttribute_41 Attributes)
  (connect DeleteAttributes_4 Out AddAttribute_41 Attributes)
  (connect MergeAttributes_20 Out AddAttribute_41 Attributes)
  (connect MergeAttributes_22 Out AddAttribute_41 Attributes)
  (connect MergeAttributes_24 Out AddAttribute_41 Attributes)
  (connect MergeAttributes_26 Out AddAttribute_41 Attributes)
  (connect MergeAttributes_16 Out AddAttribute_41 In)
  (connect GetEntriesCount_32 Out AttributeCast_0 In)
  (connect Branch_30 Output%20B AttributeFilter_0 In)
  (connect AddAttribute_41 Out AttributeFilter_45 In)
  (connect MergeAttributes_28 Out AttributeRename_3 In)
  (connect AttributeFilter_0 InsideFilter Branch_1 In)
  (connect AttributeFilter_0 InsideFilter Branch_5 In)
  (connect AttributeFilter_0 InsideFilter Branch_7 In)
  (connect AttributeFilter_0 InsideFilter Branch_9 In)
  (connect AttributeFilter_0 InsideFilter Branch_18 In)
  (connect AttributeFilter_0 InsideFilter Branch_21 In)
  (connect AttributeCast_0 Out Branch_30 bOutputToB)
  (connect DefaultInputNode In Branch_30 In)
  (connect GetPropertyFromObjectPath_3 Out CopyAttributes_0 Source)
  (connect AddAttribute_1 Out CopyAttributes_0 Target)
  (connect MakeTransformAttribute_3 Out DeleteAttributes_4 In)
  (connect DefaultInputNode In GetEntriesCount_32 In)
  (connect AttributeFilter_0 InsideFilter GetPropertyFromObjectPath_0 In)
  (connect Branch_5 Output%20B GetPropertyFromObjectPath_1 In)
  (connect Branch_7 Output%20B GetPropertyFromObjectPath_2 In)
  (connect AttributeFilter_0 InsideFilter GetPropertyFromObjectPath_3 In)
  (connect Branch_1 Output%20B GetPropertyFromObjectPath_6 In)
  (connect Branch_18 Output%20B GetPropertyFromObjectPath_8 In)
  (connect Branch_21 Output%20B GetPropertyFromObjectPath_10 In)
  (connect Branch_9 Output%20B GetPropertyFromObjectPath_12 In)
  (connect MergeAttributes_18 Out MakeTransformAttribute_3 Rotation)
  (connect MergeAttributes_18 Out MakeTransformAttribute_3 Scale)
  (connect MergeAttributes_18 Out MakeTransformAttribute_3 Translation)
  (connect CopyAttributes_0 Out MergeAttributes_16 In)
  (connect GetPropertyFromObjectPath_1 Out MergeAttributes_18 In)
  (connect GetPropertyFromObjectPath_6 Out MergeAttributes_20 In)
  (connect GetPropertyFromObjectPath_8 Out MergeAttributes_22 In)
  (connect GetPropertyFromObjectPath_10 Out MergeAttributes_24 In)
  (connect GetPropertyFromObjectPath_12 Out MergeAttributes_26 In)
  (connect GetPropertyFromObjectPath_2 Out MergeAttributes_28 In)
  (output Out AttributeFilter_45.InsideFilter)
  (output Out Branch_30.Output%20A))
