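; PCG subgraph for Biome Core ground scatter that turns an entry's Weight into a cumulative weight.
; It filters All to elements whose $Index is lower than the input's index attribute, then sums their Weight.
; A default constant Weight is merged in first, so the sum still has data when no element precedes the entry.
; That sum is combined with the input's Weight by the default Add operation to give the running total.
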(graph GroundScatterAccumulateWeights
  (inputs (All :type Param))
  (bind AttributeFilter_0 (Filter%20Attribute%20Elements :Operator Lesser :TargetAttribute $Index :ThresholdAttribute index))
  (bind CreateAttribute_0 (Create%20Constant :OutputTarget Weight))
  (bind MergeAttributes_0 (Merge%20Attributes))
  (bind AttributeReduce_0 (Attribute%20Reduce :InputSource Weight :Operation Sum :OutputAttributeName Weight))
  (bind AttributeMathsOp_0 (Attribute%20Maths%20Op :InputSource1 Weight :InputSource2 Weight))
  (connect DefaultInputNode In AttributeFilter_0 Filter)
  (connect DefaultInputNode All AttributeFilter_0 In)
  (connect DefaultInputNode In AttributeMathsOp_0 InA)
  (connect AttributeReduce_0 Out AttributeMathsOp_0 InB)
  (connect MergeAttributes_0 Out AttributeReduce_0 In)
  (connect CreateAttribute_0 Out MergeAttributes_0 In)
  (connect AttributeFilter_0 InsideFilter MergeAttributes_0 In2)
  (output Out AttributeMathsOp_0))
