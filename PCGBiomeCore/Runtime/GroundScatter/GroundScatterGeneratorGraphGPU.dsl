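; PCG graph for biome ground scatter that generates points per tile on the GPU and spawns static meshes from biome assets.
; The UseHeightRVT root parameter branches terrain input to either the Landscape or height virtual texture point generator subgraph, both fed tile influence points.
; A Custom HLSL kernel takes the gathered points with biome data and outputs a Mesh attribute.
; The mesh spawner selects by that attribute, with shadows disabled.
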(graph GroundScatterGeneratorGraphGPU
  (params (DensityFilter :type Double :default 0.5) (RandomNoiseScale :type Double :default 1e+02) (RandomPositionOffset :type Double :default 1e+02) (PointsScaleMin :type Double :default 0.1) (PointsScaleMax :type Double :default 1.0) (SpatialNoiseSize :type Double :default 1e+02) (SpatialNoiseStrength :type Double :default 1.0) (SlopeDensityMin :type Double) (SlopeDensityMax :type Double :default 1.0) (ScaleByDensityMin :type Double) (ScaleByDensityMax :type Double :default 1.0) (InfluencePointsRadiusMultiplier :type Double :default 1.0) (InfluencePointsFalloffDistance :type Double :default 5e+02) (RandomNoiseSeed :type Int :default 42) (SpatialNoiseSeed :type Int :default 42))
  (inputs (Landscape :type Landscape) (WorldHeightRVT :type VirtualTexture) (TileInfluencePoints :type Point) (RootGraphParams :type Param) (BiomeAssets :type Param) (TileBiomePoints :type Param))
  (bind DeleteAttributes_2 (Delete%20Attributes :Operation KeepSelectedAttributes :SelectedAttributes "UseHeightRVT"))
  (bind Branch_5 (Branch))
  (bind CreateConstant_0 (Create%20Constant :AttributeTypes (struct :Type SoftObjectPath :SoftObjectPathValue (struct :AssetPath (struct :PackageName /PCGBiomeCore/Runtime/GroundScatter/PointGeneratorLS :AssetName PointGeneratorLS)))))
  (bind Subgraph_0 (Subgraph :Graph /PCGBiomeCore/Runtime/GroundScatter/PointGeneratorLS.PointGeneratorLS :LocalGraphParams ))
  (bind Branch_3 (Branch))
  (bind CreateConstant_1 (Create%20Constant :AttributeTypes (struct :Type SoftObjectPath :SoftObjectPathValue (struct :AssetPath (struct :PackageName /PCGBiomeCore/Runtime/GroundScatter/PointGeneratorRVT :AssetName PointGeneratorRVT)))))
  (bind Subgraph_1 (Subgraph :Graph /PCGBiomeCore/Runtime/GroundScatter/PointGeneratorRVT.PointGeneratorRVT :LocalGraphParams ))
  (bind Gather_2 (Gather))
  ;[0.15 0.15 0.15 0.5] Out
  (bind CustomHLSL_2 (Custom%20HLSL :InputPins [(struct :Label GeneratorPoints :AllowedTypes (struct :Ids [(struct :Struct /Script/PCG.PCGDataTypeInfoPoint)]) :PinStatus Required :bAllowMultipleConnections false) (struct :Label RootGraphParams :AllowedTypes (struct :Ids [(struct :Struct /Script/PCG.PCGDataTypeInfoParam)]) :bAllowMultipleData false :bAllowMultipleConnections false) (struct :Label BiomeAssets :AllowedTypes (struct :Ids [(struct :Struct /Script/PCG.PCGDataTypeInfoParam)]) :bAllowMultipleData false :bAllowMultipleConnections false) (struct :Label TileBiomePoints :AllowedTypes (struct :Ids [(struct :Struct /Script/PCG.PCGDataTypeInfoParam)]) :bAllowMultipleData false :bAllowMultipleConnections false)] :OutputPins [(struct :PropertiesGPU (struct :PinsToInititalizeFrom [GeneratorPoints] :bAutoInitializeOutput false :CreatedKernelAttributeKeys [(struct :Identifier (struct :Name Mesh) :Type StringKey :Name Mesh)]) :Label Out :AllowedTypes (struct :Ids [(struct :Struct /Script/PCG.PCGDataTypeInfoPoint)]))]))
  (bind StaticMeshSpawner_0 (Static%20Mesh%20Spawner :bApplyMeshBoundsToPoints false :MeshSelectorParameters (class:/Script/PCG.PCGMeshSelectorByAttribute :AttributeName Mesh :MaterialOverrideAttributes [Material] :TemplateDescriptor (struct :bAffectDynamicIndirectLighting false :bAffectDistanceFieldLighting false :bCastDynamicShadow false :bCastStaticShadow false :bVisibleInRayTracing false :bIncludeInHLOD false :bForceNavigationObstacle false)) :MeshSelectorType /Script/PCG.PCGMeshSelectorByAttribute))
  (connect DeleteAttributes_2 Out Branch_3 bOutputToB)
  (connect DefaultInputNode WorldHeightRVT Branch_3 In)
  (connect DeleteAttributes_2 Out Branch_5 bOutputToB)
  (connect DefaultInputNode Landscape Branch_5 In)
  (connect DefaultInputNode BiomeAssets CustomHLSL_2 BiomeAssets)
  (connect Gather_2 Out CustomHLSL_2 GeneratorPoints)
  (connect DefaultInputNode RootGraphParams CustomHLSL_2 RootGraphParams)
  (connect DefaultInputNode TileBiomePoints CustomHLSL_2 TileBiomePoints)
  (connect DefaultInputNode RootGraphParams DeleteAttributes_2 In)
  (connect Subgraph_0 Point%20Generator%20Out Gather_2 In)
  (connect Subgraph_1 Point%20Generator%20Out Gather_2 In2)
  (connect CustomHLSL_2 Out StaticMeshSpawner_0 In)
  (connect Branch_5 Output%20A Subgraph_0 Landscape)
  (connect CreateConstant_0 Out Subgraph_0 SubgraphOverride)
  (connect DefaultInputNode TileInfluencePoints Subgraph_0 TileInfluencePoints)
  (connect CreateConstant_1 Out Subgraph_1 SubgraphOverride)
  (connect DefaultInputNode TileInfluencePoints Subgraph_1 TileInfluencePoints)
  (connect Branch_3 Output%20B Subgraph_1 WorldHeightRVT))
