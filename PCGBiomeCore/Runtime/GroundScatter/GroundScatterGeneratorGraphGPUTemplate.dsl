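; GroundScatterGeneratorGraphGPUTemplate is a PCG graph template that, by its name, serves as the starting point for GPU ground scatter generators in PCG Biome Core.
; It declares inputs for biome assets, landscape, world-height virtual texture, tile biome points, tile influence points and root graph parameters.
; A PointsPerSqm integer parameter exposes scatter density.
; The graph contains no nodes, so it defines only the interface.
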(graph GroundScatterGeneratorGraphGPUTemplate
  (params (PointsPerSqm :type Int :default 100))
  (inputs (BiomeAssets :type Param) (Landscape :type Landscape) (WorldHeightRVT :type VirtualTexture) (TileBiomePoints :type Param) (TileInfluencePoints :type Point) (RootGraphParams :type Param)))
