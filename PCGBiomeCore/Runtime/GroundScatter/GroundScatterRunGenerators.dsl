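; PCG subgraph for biome ground scatter that runs the generator graph named by the incoming biome assets.
; It takes an entry from AssetsByGeneratorGraph by index, keeps only its GeneratorGraph attribute, and feeds that as the subgraph override.
; The subgraph node defaults to GroundScatterGeneratorGraphGPUTemplate and receives the full asset set, landscape, world height RVT, tile biome and influence points, and root graph parameters.
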(graph GroundScatterRunGenerators
  (inputs (AssetsByGeneratorGraph :type Param) (Landscape :type Landscape) (WorldHeightRVT :type VirtualTexture) (TileBiomePoints :type Param) (TileInfluencePoints :type Point) (GraphParams :type Param))
  (bind GetAttributeFromIndex_0 (Get%20Attribute%20Set%20from%20Index))
  (bind DeleteAttributes_0 (Delete%20Attributes :Operation KeepSelectedAttributes :SelectedAttributes "GeneratorGraph"))
  (bind Subgraph_3 (Subgraph :Graph /PCGBiomeCore/Runtime/GroundScatter/GroundScatterGeneratorGraphGPUTemplate.GroundScatterGeneratorGraphGPUTemplate))
  (reroute BiomeAssets AssetsByGeneratorGraph)
  (connect GetAttributeFromIndex_0 Out DeleteAttributes_0 In)
  (connect BiomeAssets Out GetAttributeFromIndex_0 In)
  (connect BiomeAssets Out Subgraph_3 BiomeAssets)
  (connect DefaultInputNode Landscape Subgraph_3 Landscape)
  (connect DefaultInputNode GraphParams Subgraph_3 RootGraphParams)
  (connect DeleteAttributes_0 Out Subgraph_3 SubgraphOverride)
  (connect DefaultInputNode TileBiomePoints Subgraph_3 TileBiomePoints)
  (connect DefaultInputNode TileInfluencePoints Subgraph_3 TileInfluencePoints)
  (connect DefaultInputNode WorldHeightRVT Subgraph_3 WorldHeightRVT))
