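; PCG subgraph in Biome Core's ground scatter that turns a set of entry weights into a normalized, accumulated weight distribution.
; It divides each Weight by the summed total, then copies each entry's $Index into an index attribute.
; Partitioning by $Index feeds entries one by one into a GroundScatterAccumulateWeights loop, alongside the full set as All.
; The loop results are merged into a single attribute set output.
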(graph GroundScatterWeightsDistributionLoop
  (bind AttributeReduce_1 (Attribute%20Reduce :InputSource Weight :Operation Sum))
  (bind AttributeMathsOp_0 (Attribute%20Maths%20Op :InputSource1 Weight :Operation Divide))
  (bind CopyAttributes_0 (Copy%20Attributes :InputSource $Index :OutputTarget index))
  (bind AttributePartition_1 (Attribute%20Partition :PartitionAttributeSelectors [$Index]))
  (bind Loop_0 (Loop :Graph /PCGBiomeCore/Runtime/GroundScatter/GroundScatterAccumulateWeights.GroundScatterAccumulateWeights))
  (bind MergeAttributes_0 (Merge%20Attributes))
  (connect DefaultInputNode In AttributeMathsOp_0 InA)
  (connect AttributeReduce_1 Out AttributeMathsOp_0 InB)
  (connect CopyAttributes_0 Out AttributePartition_1 In)
  (connect DefaultInputNode In AttributeReduce_1 In)
  (connect AttributeMathsOp_0 Out CopyAttributes_0 Source)
  (connect AttributeMathsOp_0 Out CopyAttributes_0 Target)
  (connect CopyAttributes_0 Out Loop_0 All)
  (connect AttributePartition_1 Out Loop_0 In)
  (connect Loop_0 Out MergeAttributes_0 In)
  (output Out MergeAttributes_0))
