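; PCG subgraph in the biome core's ground scatter runtime that generates points from a landscape, tile influence points and local graph parameters.
; All three inputs feed a single Custom HLSL node running as a point-generator kernel with a fixed count of 262,144 elements.
; The influence points enter the kernel as Influencers, the local parameters pin is required, and the kernel's points are the graph's only output.
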(graph PointGeneratorLS
  (inputs (Landscape :type Landscape) (TileInfluencePoints :type Point) (LocalGraphParams :type Param))
  (outputs (Point%20Generator%20Out :type Point))
  (bind CustomHLSL_2 (Custom%20HLSL :InputPins [(struct :Label Landscape :AllowedTypes (struct :Ids [(struct :Struct /Script/PCG.PCGDataTypeInfoLandscape)]) :bAllowMultipleData false :bAllowMultipleConnections false) (struct :Label Influencers :AllowedTypes (struct :Ids [(struct :Struct /Script/PCG.PCGDataTypeInfoPoint)]) :bAllowMultipleConnections false) (struct :Label LocalGraphParams :AllowedTypes (struct :Ids [(struct :Struct /Script/PCG.PCGDataTypeInfoParam)]) :bAllowMultipleData false :PinStatus Required :bAllowMultipleConnections false)] :KernelType PointGenerator :NumElements 262144))
  (connect DefaultInputNode TileInfluencePoints CustomHLSL_2 Influencers)
  (connect DefaultInputNode Landscape CustomHLSL_2 Landscape)
  (connect DefaultInputNode LocalGraphParams CustomHLSL_2 LocalGraphParams)
  (output Point%20Generator%20Out CustomHLSL_2))
