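; PCG subgraph in Biome Core's runtime ground scatter that generates points on the GPU, named for its runtime virtual texture input.
; A single Custom HLSL point-generator kernel runs a fixed count of 262,144 elements, taking a world-height RVT, tile influence points as Influencers and a required local parameter set.
; Output auto-initialization is off; the kernel's points go straight to the graph output.
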(graph PointGeneratorRVT
  (inputs (WorldHeightRVT :type VirtualTexture) (TileInfluencePoints :type Point) (LocalGraphParams :type Param))
  (outputs (Point%20Generator%20Out :type Point))
  (bind CustomHLSL_1 (Custom%20HLSL :InputPins [(struct :Label WorldHeightRVT :AllowedTypes (struct :Ids [(struct :Struct /Script/PCG.PCGDataTypeInfoVirtualTexture)]) :bAllowMultipleConnections false) (struct :Label Influencers :AllowedTypes (struct :Ids [(struct :Struct /Script/PCG.PCGDataTypeInfoPoint)]) :bAllowMultipleConnections false) (struct :Label LocalGraphParams :AllowedTypes (struct :Ids [(struct :Struct /Script/PCG.PCGDataTypeInfoParam)]) :bAllowMultipleData false :PinStatus Required :bAllowMultipleConnections false)] :KernelType PointGenerator :NumElements 262144 :OutputPins [(struct :PropertiesGPU (struct :bAutoInitializeOutput false) :Label Out :AllowedTypes (struct :Ids [(struct :Struct /Script/PCG.PCGDataTypeInfoPoint)]))]))
  (connect DefaultInputNode TileInfluencePoints CustomHLSL_1 Influencers)
  (connect DefaultInputNode LocalGraphParams CustomHLSL_1 LocalGraphParams)
  (connect DefaultInputNode WorldHeightRVT CustomHLSL_1 WorldHeightRVT)
  (output Point%20Generator%20Out CustomHLSL_1))
